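# Niagara effect for the HighTower Tomato glider: thruster-style exhaust at four spawn groups bound to bones or sockets of a user skeletal mesh.
# Startup bursts of hot core, flare, sparks and smoke give way to looping flame meshes, glow, smoke, random GPU sparks and a persistent rim light.
# User ForwardDot and RightDot feed a per-group Glider_Dot_Adjustment that scales flame, glow and smoke velocity.
# ForwardDot also drives smoke size and spark probability.
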NiagaraSystem /BRCosmetics/Gadgets/Assets/VinderTech_GliderChute/Glider_HighTower_Tomato/FX/NS_Glider_HighTower_Tomato_Player
dependencies {
 usedDataInterfaces [
  &NiagaraDataInterfaceCurve
  &NiagaraDataInterfaceColorCurve
  &NiagaraDataInterfaceArrayFloat3
  /Script/Niagara.NiagaraDataInterfaceVectorField
  &NiagaraDataInterfaceVectorCurve
 ]
 usedDynamicInputs []
 usedModules [&SystemState]
}
emitters {
 Animorphic_-_HotCoreStart_-_FlameOuterStart {
  emitterUpdateScript [
   &EmitterState(Life Cycle Mode=::NewEnumerator1, Loop Behavior=::NewEnumerator1, Loop Duration=0.1, UseLoopDelay=false)
   &SpawnBurst_Instantaneous(Spawn Count=4, Use Spawn Probability=false)
  ]
  particleSpawnScript [
   &FN_SetSpawnGroup_v2(Number of Spawn Groups=4)
   /Niagara/Modules/Spawn/Initialization/InitializeParticle(
    Color=<rgba 1.5 0.55 0 0.1>
    Lifetime=0.2
    Mesh Scale=<0.333 1 1>
    Module.Write Color=true
    Module.Write Lifetime=true
    Module.Write Mass=true
    Module.Write Material Random=false
    Module.Write Position=false
    Module.Write Scale=true
    Module.Write SpriteRotation=false
    Module.Write SpriteSize=true
    Position=&SimulationPosition()
    Sprite Size=&UniformRangedVector2D(Minimum=<100 4>, Maximum=<50 1>)
   )
   /SetVariables(
    Particles.CameraOffset_2=10
    Particles.Color_2:LinearColor=<rgba 1 0.3667 0 1>
    Particles.DynamicMaterialParameter_2:Vector4f=<1 1 1 1>
    Particles.SpriteSize_2:Vector2f=<100 100>
   ) {isSetParameters=true}
   /SetVariables(
    Particles.Color_3:LinearColor=<rgba 5 5 5 1>
    Particles.DynamicMaterialParameter_3:Vector4f=&MakeVector4(X=0.5, Y=1, Z=&UniformRangedFloat(Minimum=0.25, Maximum=0.35), W=1)
    Particles.SpriteSize_3:Vector2f=&Vector2DFromFloat(Value=&UniformRangedFloat(Minimum=50, Maximum=75))
   ) {isSetParameters=true}
   /SetVariables(
    Particles.Color_4:LinearColor=<rgba 1.5 0.7875 0.375 0.1>
    Particles.DynamicMaterialParameter_4:Vector4f=&MakeVector4(X=1, Y=5, Z=20, W=1)
   ) {isSetParameters=true}
   &CameraOffset(Camera Offset Amount=2)
   &DynamicMaterialParameters(
    Index 0 Param 1=5
    Module.Param0WriteEnabled=true
    Module.Param1WriteEnabled=false
    Module.Param2WriteEnabled=false
    Module.Param3WriteEnabled=false
   )
  ]
  particleUpdateScript [
   &ParticleState()
   &SkeletalMeshLocation(
    Apply Sampled Orientation to Mesh Particles=true
    Bone / Socket Index=$Particles.SpawnGroup
    Inherit Velocity=false
    Interpolate Mesh Sampling=false
    MaskSpawnGroup=false
    Mesh Sampling Type=::NewEnumerator5
    Skeletal Mesh=$User.Skeletal Mesh
    Socket Sampling Mode=::NewEnumerator5
   )
   &ScaleSpriteSize(Scale Factor=&MakeVector2D(X=&FloatFromCurve(FloatCurve=DataInterface() {propertyValues {Curve {keys [<at 0 0>,<at 1 2>]}}}), Y=1))
   /SetVariables(Particles.SpriteSize_3:Vector2f=&Multiply_Vector2(
     A=$Particles.Initial.SpriteSize_3
     B=&Vector2DFromFloat(Value=&FloatFromCurve(FloatCurve=DataInterface() {propertyValues {Curve {keys [<at 0 0>,<at 1 1>]}}}))
    )) {isSetParameters=true}
   &ScaleColor(
    Scale Alpha=&FloatFromCurve(FloatCurve=DataInterface() {propertyValues {Curve {keys [<at 0 1>,<at 1 0>]}}})
    Scale RGB=&MakeVectorFromLinearColorRGB(Color=&ColorFromCurve(ColorCurve=DataInterface() {propertyValues {
        AlphaCurve {keys []}
        BlueCurve {keys [<at 0 0.25 RCIM_Cubic>,<at 0.5 0.5>,<at 1 0.875 RCIM_Cubic>]}
        GreenCurve {keys [<at 0 0.525 RCIM_Cubic>,<at 0.5 0.6833>,<at 1 0.9208 RCIM_Cubic>]}
        RedCurve {keys [<at 0 1 RCIM_Cubic>,<at 0.5 1>,<at 1 1 RCIM_Cubic>]}
       }}))
    ScaleA=true
    ScaleRGB=true
   )
   &ScaleColor(
    Scale RGB=&VectorFromFloat(Value=&FloatFromCurve(FloatCurve=DataInterface() {propertyValues {Curve {keys [
          <at 0 100 RCIM_Cubic>
          {arriveTangent=-33.2952, interpMode=RCIM_Cubic, leaveTangent=-33.2924, tangentMode=RCTM_User, time=0.25, value=10}
          <at 1 0 RCIM_Cubic>
         ]}}}))
    ScaleA=true
    ScaleRGB=true
   ) {module=ScaleColor001}
   /SetVariables(
    Particles.Color_2:LinearColor=&MultiplyLinearColor(
     Linear Color=$Particles.Initial.Color_2
     Scale Factor=&MakeVector4_FromLinearColor(LinearColor=&MakeLinearColorFromVectorAndFloat(
       "Vector (RGB)"=&MakeVectorFromLinearColorRGB(Color=&ColorFromCurve(ColorCurve=DataInterface() {propertyValues {
           AlphaCurve {keys []}
           BlueCurve {keys [<at 0 0.25 RCIM_Cubic>,<at 0.5 0.5>,<at 1 0.875 RCIM_Cubic>]}
           GreenCurve {keys [<at 0 0.525 RCIM_Cubic>,<at 0.5 0.6833>,<at 1 0.9208 RCIM_Cubic>]}
           RedCurve {keys [<at 0 1 RCIM_Cubic>,<at 0.5 1>,<at 1 1 RCIM_Cubic>]}
          }}))
       "Float (Alpha)"=&FloatFromCurve(FloatCurve=DataInterface() {propertyValues {Curve {keys [<at 0 1>,<at 1 0>]}}})
      ))
    )
    Particles.Color_3:LinearColor=&MultiplyLinearColor(
     Linear Color=$Particles.Initial.Color_3
     Scale Factor=&MakeVector4_FromLinearColor(LinearColor=&MakeLinearColorFromVectorAndFloat(
       "Vector (RGB)"=&MakeVectorFromLinearColorRGB(Color=&ColorFromCurve(ColorCurve=DataInterface() {propertyValues {
           AlphaCurve {keys []}
           BlueCurve {keys [<at 0 0.25 RCIM_Cubic>,<at 0.5 0.5>,<at 1 0.875 RCIM_Cubic>]}
           GreenCurve {keys [<at 0 0.525 RCIM_Cubic>,<at 0.5 0.6833>,<at 1 0.9208 RCIM_Cubic>]}
           RedCurve {keys [<at 0 1 RCIM_Cubic>,<at 0.5 1>,<at 1 1 RCIM_Cubic>]}
          }}))
       "Float (Alpha)"=&FloatFromCurve(FloatCurve=DataInterface() {propertyValues {Curve {keys [<at 0 1>,<at 1 0>]}}})
      ))
    )
    Particles.Color_4:LinearColor=&MultiplyLinearColor(
     Linear Color=$Particles.Initial.Color_4
     Scale Factor=&MakeVector4_FromLinearColor(LinearColor=&MakeLinearColorFromVectorAndFloat(
       "Vector (RGB)"=&MakeVectorFromLinearColorRGB(Color=&ColorFromCurve(ColorCurve=DataInterface() {propertyValues {
           AlphaCurve {keys []}
           BlueCurve {keys [<at 0 0.25 RCIM_Cubic>,<at 0.5 0.5>,<at 1 0.875 RCIM_Cubic>]}
           GreenCurve {keys [<at 0 0.525 RCIM_Cubic>,<at 0.5 0.6833>,<at 1 0.9208 RCIM_Cubic>]}
           RedCurve {keys [<at 0 1 RCIM_Cubic>,<at 0.5 1>,<at 1 1 RCIM_Cubic>]}
          }}))
       "Float (Alpha)"=&FloatFromCurve(FloatCurve=DataInterface() {propertyValues {Curve {keys [<at 0 1>,<at 1 0>]}}})
      ))
    )
   ) {isSetParameters=true}
   /SetVariables(
    Particles.Color_2:LinearColor=&MultiplyLinearColor(
     Linear Color=$Particles.Color_2
     Scale Factor=&MakeFloat4FromVectorAndFloat(
      "Vector (RGB)"=&VectorFromFloat(Value=&FloatFromCurve(FloatCurve=DataInterface() {propertyValues {Curve {keys [<at 0 1 RCIM_Cubic>, {arriveTangent=-1, interpMode=RCIM_Cubic, leaveTangent=-1, time=0.1, value=6}, <at 1 0 RCIM_Cubic>]}}}))
      "Float (Alpha)"=1
     )
    )
    Particles.Color_4:LinearColor=&MultiplyLinearColor(
     Linear Color=$Particles.Initial.Color_4
     Scale Factor=&MakeFloat4FromVectorAndFloat(
      "Vector (RGB)"=&VectorFromFloat(Value=&FloatFromCurve(FloatCurve=DataInterface() {propertyValues {Curve {keys [
            <at 0 100 RCIM_Cubic>
            {arriveTangent=-36.3579, interpMode=RCIM_Cubic, leaveTangent=-36.3595, tangentMode=RCTM_User, time=0.25, value=10}
            <at 1 1 RCIM_Cubic>
           ]}}}))
      "Float (Alpha)"=1
     )
    )
   ) {isSetParameters=true}
   &ScaleMeshSize(Scale Factor=&VectorFromCurve(VectorCurve=DataInterface() {propertyValues {
       XCurve {keys [
         {arriveTangent=2.9897, interpMode=RCIM_Cubic, leaveTangent=2.9897, tangentMode=RCTM_User, time=0, value=0}
         {arriveTangent=3.9553, interpMode=RCIM_Cubic, leaveTangent=3.9553, tangentMode=RCTM_User, time=0.2, value=0.6}
         <at 1 2 RCIM_Cubic>
        ]}
       YCurve {keys [
         {arriveTangent=14.017, interpMode=RCIM_Cubic, leaveTangent=14.017, tangentMode=RCTM_User, time=0, value=0}
         <at 0.2 1 RCIM_Cubic>
         <at 1 0 RCIM_Cubic>
        ]}
       ZCurve {keys [
         {arriveTangent=13.3462, interpMode=RCIM_Cubic, leaveTangent=13.3462, tangentMode=RCTM_User, time=0, value=0}
         <at 0.2 1 RCIM_Cubic>
         <at 1 0 RCIM_Cubic>
        ]}
      }}))
  ]
  rendererClasses [&NiagaraSpriteRendererProperties, &NiagaraMeshRendererProperties]
  renderers [
   {rendererClass=&NiagaraSpriteRendererProperties}
   {rendererClass=&NiagaraSpriteRendererProperties, rendererIndex=1}
   {rendererClass=&NiagaraSpriteRendererProperties, rendererIndex=2}
   {rendererClass=&NiagaraMeshRendererProperties, rendererIndex=3}
  ]
 }
 Comet_Flame_Meshes {
  emitterUpdateScript [
   &EmitterState(Delay First Loop Only=true, Life Cycle Mode=::NewEnumerator1, Loop Delay=0.2, Loop Duration=0.1, UseLoopDelay=true)
   &SpawnBurst_Instantaneous(Spawn Count=4, Use Spawn Probability=false)
  ]
  particleSpawnScript [
   &FN_SetSpawnGroup_v2(Number of Spawn Groups=4)
   &InitializeParticle(
    Color=<rgba 1.5 0.375 0.15 5>
    Lifetime=0.2
    Mesh Scale=<-0.0175 0.035 0.035>
    Module.Write Color=true
    Module.Write Lifetime=true
    Module.Write Mass=true
    Module.Write Material Random=false
    Module.Write Position=false
    Module.Write Scale=true
    Module.Write SpriteRotation=false
    Module.Write SpriteSize=false
    Position=&SimulationPosition()
    Sprite Size=<5 5>
   )
   /SetVariables(Particles.Color_2:LinearColor=<rgba 20 12.5 2 0.25>, Particles.Scale_2:Vector3f=<0.0165 0.025 0.025>) {isSetParameters=true}
   &DynamicMaterialParameters(
    Index 0 Param 1=5
    Index 0 Param 2=5
    Module.Param0WriteEnabled=true
    Module.Param1WriteEnabled=true
    Module.Param2WriteEnabled=true
    Module.Param3WriteEnabled=true
   )
  ]
  particleUpdateScript [
   &ParticleState()
   &SkeletalMeshLocation(
    Apply Sampled Orientation to Mesh Particles=true
    Bone / Socket Index=$Particles.SpawnGroup
    Inherit Velocity=false
    Interpolate Mesh Sampling=false
    MaskSpawnGroup=false
    Mesh Sampling Type=::NewEnumerator5
    Skeletal Mesh=$User.Skeletal Mesh
    Socket Sampling Mode=::NewEnumerator5
   )
   &ScaleColor(
    Scale Alpha=&Lerp_Float(A=1, B=0, Alpha=$Particles.NormalizedAge)
    Scale RGB=&MakeVectorFromLinearColorRGB(Color=&ColorFromCurve(ColorCurve=DataInterface() {propertyValues {
        AlphaCurve {keys []}
        BlueCurve {keys [<at 0 0.25 RCIM_Cubic>,<at 0.492 0.5>,<at 1 0.875 RCIM_Cubic>]}
        GreenCurve {keys [<at 0 0.4375 RCIM_Cubic>,<at 0.492 0.625>,<at 1 0.9062 RCIM_Cubic>]}
        RedCurve {keys [<at 0 1 RCIM_Cubic>,<at 0.492 1>,<at 1 1 RCIM_Cubic>]}
       }}))
    ScaleA=true
    ScaleRGB=true
   )
   &ScaleMeshSize(Scale Factor=&MakeVector(
     X=&FloatFromCurve(FloatCurve=DataInterface() {propertyValues {Curve {keys [<at 0 0.75 RCIM_Cubic>, {arriveTangent=-0.25, interpMode=RCIM_Cubic, leaveTangent=-0.25, time=0.5, value=2.5}, <at 1 0.5 RCIM_Cubic>]}}})
     Y=&FloatFromCurve(FloatCurve=DataInterface() {propertyValues {Curve {keys [<at 0.5 1 RCIM_Cubic>,<at 1 1.75 RCIM_Cubic>]}}})
     Z=&FloatFromCurve(FloatCurve=DataInterface() {propertyValues {Curve {keys [<at 0.5 1 RCIM_Cubic>,<at 1 1.75 RCIM_Cubic>]}}})
    ))
   &ScaleMeshSize(Scale Factor=&MakeVector(
     X=&self:Glider_Dot_Adjustment(
      Forward Dot:NiagaraFloat=$User.ForwardDot
      Max Value:NiagaraFloat=1.5
      Min Value:NiagaraFloat=1
      Right Dot:NiagaraFloat=$User.RightDot
      Spawn Group:NiagaraInt32=$Particles.SpawnGroup
     )
     Y=1
     Z=1
    )) {module=ScaleMeshSize001}
   /SetVariables(Particles.Scale_2:Vector3f=&Multiply_Vector(
     A=$Particles.Initial.Scale_2
     B=&MakeVector(
      X=&FloatFromCurve(FloatCurve=DataInterface() {propertyValues {Curve {keys [<at 0 0.75 RCIM_Cubic>, {arriveTangent=-0.25, interpMode=RCIM_Cubic, leaveTangent=-0.25, time=0.5, value=2.5}, <at 1 0.5 RCIM_Cubic>]}}})
      Y=&FloatFromCurve(FloatCurve=DataInterface() {propertyValues {Curve {keys [<at 0.5 1 RCIM_Cubic>,<at 1 1.75 RCIM_Cubic>]}}})
      Z=&FloatFromCurve(FloatCurve=DataInterface() {propertyValues {Curve {keys [<at 0.5 1 RCIM_Cubic>,<at 1 1.75 RCIM_Cubic>]}}})
     )
    )) {isSetParameters=true}
   /SetVariables(Particles.Scale_2:Vector3f=&Multiply_Vector(
     A=$Particles.Scale_2
     B=&MakeVector(
      X=&self:Glider_Dot_Adjustment(
       Forward Dot:NiagaraFloat=$User.ForwardDot
       Max Value:NiagaraFloat=1.5
       Min Value:NiagaraFloat=0.5
       Right Dot:NiagaraFloat=$User.RightDot
       Spawn Group:NiagaraInt32=$Particles.SpawnGroup
      )
      Y=1
      Z=1
     )
    )) {isSetParameters=true}
  ]
  rendererClasses [&NiagaraMeshRendererProperties]
  renderers [
   {rendererClass=&NiagaraMeshRendererProperties, rendererIndex=0}
   {rendererClass=&NiagaraMeshRendererProperties, rendererIndex=1}
  ]
 }
 Flame_Glow {
  emitterUpdateScript [
   &EmitterState(Delay First Loop Only=true, Life Cycle Mode=::NewEnumerator1, Loop Delay=0.2, Loop Duration=0.1, UseLoopDelay=true)
   &SpawnRate(SpawnRate=80, Use Spawn Probability=false)
  ]
  particleSpawnScript [
   &FN_SetSpawnGroup_v2(Number of Spawn Groups=4)
   &InitializeParticle(
    Color=<rgba 1 0.325 0.1 0.05>
    Lifetime=0.2
    Module.Write Color=true
    Module.Write Lifetime=true
    Module.Write Mass=true
    Module.Write Material Random=true
    Module.Write Position=false
    Module.Write Scale=false
    Module.Write SpriteRotation=false
    Module.Write SpriteSize=true
    Position=&SimulationPosition()
    Sprite Size=<7 15>
   )
   &SkeletalMeshLocation(
    Apply To Particle Position=false
    Bone / Socket Index=$Particles.SpawnGroup
    Inherit Velocity=false
    Interpolate Mesh Sampling=false
    MaskSpawnGroup=false
    Mesh Sampling Type=::NewEnumerator5
    Sampled Position Offset=&MakeVector(
     X=&FloatFromCurve(
      FloatCurve=DataInterface() {propertyValues {Curve {keys [<at 0 0>,<at 1 1>]}}}
      Scale Curve=&Lerp_Float(A=12.5, B=12, Alpha=$User.ForwardDot)
     )
     Y=0
     Z=0
    )
    Skeletal Mesh=$User.Skeletal Mesh
    Socket Sampling Mode=::NewEnumerator5
   ) {module=SkeletalMeshLocation001}
   &AlignSpriteToMeshOrientation(Mesh Orientation Relative Sprite Alignment Vector=&RotateVectorByQuaternion(Vector To Rotate=<1 0 0>, Quaternion=$Output.SkeletalMeshLocation001.SampledRotation))
  ]
  particleUpdateScript [
   &ParticleState()
   &SkeletalMeshLocation(
    Bone / Socket Index=$Particles.SpawnGroup
    Inherit Velocity=false
    Interpolate Mesh Sampling=false
    MaskSpawnGroup=false
    Mesh Sampling Type=::NewEnumerator5
    Sampled Position Offset=&MakeVector(
     X=&FloatFromCurve(
      FloatCurve=DataInterface() {propertyValues {Curve {keys [<at 0 0>,<at 1 1>]}}}
      Scale Curve=&self:Glider_Dot_Adjustment(
       Forward Dot:NiagaraFloat=$User.ForwardDot
       Max Value:NiagaraFloat=15
       Min Value:NiagaraFloat=7
       Right Dot:NiagaraFloat=$User.RightDot
       Spawn Group:NiagaraInt32=$Particles.SpawnGroup
      )
     )
     Y=0
     Z=0
    )
    Sampled Position Offset Coordinate Space:ENiagaraMeshTransforms=::NewEnumerator3
    Skeletal Mesh=$User.Skeletal Mesh
    Socket Sampling Mode=::NewEnumerator5
   )
   &ScaleColor(
    Scale Alpha=&Lerp_Float(A=1, B=0, Alpha=$Particles.NormalizedAge)
    Scale RGB=&MakeVectorFromLinearColorRGB(Color=&ColorFromCurve(ColorCurve=DataInterface() {propertyValues {
        AlphaCurve {keys []}
        BlueCurve {keys [<at 0 2.5 RCIM_Cubic>,<at 0.492 2.5>,<at 1 0.875 RCIM_Cubic>]}
        GreenCurve {keys [<at 0 4.375 RCIM_Cubic>,<at 0.492 3.125>,<at 1 0.9062 RCIM_Cubic>]}
        RedCurve {keys [<at 0 10 RCIM_Cubic>,<at 0.492 5>,<at 1 1 RCIM_Cubic>]}
       }}))
    ScaleA=true
    ScaleRGB=true
   )
   &ScaleSpriteSize(Scale Factor=&MakeVector2D(
     X=&FloatFromCurve(
      FloatCurve=DataInterface() {propertyValues {Curve {keys [<at 0 1>,<at 1 0>]}}}
      Scale Curve=&self:Glider_Dot_Adjustment(
       Forward Dot:NiagaraFloat=$User.ForwardDot
       Max Value:NiagaraFloat=1
       Min Value:NiagaraFloat=0.75
       Right Dot:NiagaraFloat=$User.RightDot
       Spawn Group:NiagaraInt32=$Particles.SpawnGroup
      )
     )
     Y=&self:Glider_Dot_Adjustment(
      Forward Dot:NiagaraFloat=$User.ForwardDot
      Max Value:NiagaraFloat=1
      Min Value:NiagaraFloat=0.75
      Right Dot:NiagaraFloat=$User.RightDot
      Spawn Group:NiagaraInt32=$Particles.SpawnGroup
     )
    ))
  ]
  rendererClasses [&NiagaraSpriteRendererProperties]
  renderers [{rendererClass=&NiagaraSpriteRendererProperties}]
 }
 Random_Sparks {
  simTarget=GPUComputeSim
  emitterUpdateScript [
   &EmitterState(
    Life Cycle Mode=::NewEnumerator1
    Loop Delay=&UniformRangedFloat(Minimum=0, Maximum=2)
    Loop Duration=&UniformRangedFloat(Minimum=0.1, Maximum=0.5)
    Recalculate Duration Each Loop=true
    UseLoopDelay=true
   )
   &SpawnRate(Spawn Probability=&Lerp_Float(A=0, B=1, Alpha=$User.ForwardDot), SpawnRate=10, Use Spawn Probability=true)
  ]
  particleSpawnScript [
   &FN_SetSpawnGroup_v2(Number of Spawn Groups=4)
   &InitializeParticle(
    Color=<rgba 10 6.1806 4.7917 1>
    Lifetime=&UniformRangedFloat(Minimum=0.12, Maximum=0.4)
    Module.Write Color=true
    Module.Write Lifetime=true
    Module.Write Mass=true
    Module.Write Material Random=true
    Module.Write Position=false
    Module.Write Scale=false
    Module.Write SpriteRotation=false
    Module.Write SpriteSize=true
    Position=&SimulationPosition()
    Sprite Size=&UniformRangedVector2D(Minimum=<0.7 0.5>, Maximum=<1.2 1.5>)
   )
   /Niagara/Modules/Spawn/Location/SphereLocation(
    Hemisphere Z=true
    MaskSpawnGroup=false
    Module.Override Randomness=false
    Non Uniform Scale=<1 1 -1>
    Sphere Radius=&UniformRangedFloat(Minimum=5, Maximum=10)
    Surface Only=false
   )
   &AddVelocity(Velocity=&Multiply_Vector(
     A=$Output.SphereLocation.SphereNormal
     B=&VectorFromFloat(Value=&UniformRangedFloat(Minimum=12, Maximum=25))
    ))
   &AddVelocity(Velocity=&RotateVectorByQuaternion(
     Vector To Rotate=&UniformRangedVector(Minimum=<100 -100 -100>, Maximum=<200 100 100>)
     Quaternion=$Output.SkeletalMeshLocation.SampledRotation
    )) {module=AddVelocity001}
  ]
  particleUpdateScript [
   &ParticleState()
   &SkeletalMeshLocation(
    Bone / Socket Index=$Particles.SpawnGroup
    Inherit Velocity=false
    Interpolate Mesh Sampling=false
    MaskSpawnGroup=false
    Mesh Sampling Type=::NewEnumerator5
    Sampled Position Offset Coordinate Space:ENiagaraMeshTransforms=::NewEnumerator3
    Skeletal Mesh=$User.Skeletal Mesh
    Socket Sampling Mode=::NewEnumerator5
   )
   &AccelerationForce(Acceleration=<0 0 -1200>)
   /Niagara/VectorFields/SampleVectorField(
    Sampled Vector Scale=0.4
    VectorField=DataInterface() {propertyValues {Field=/Game/Effects/VectorFields/Noisy_Turbulence_103, bTileX=true, bTileY=true, bTileZ=true}}
   )
   /Niagara/VectorFields/ApplyVectorField(Vector Field Intensity as Velocity=250, Vector Field Velocity Tightness=0.15)
   &SolveForcesAndVelocity(Clamp Velocity=false, Limit Acceleration=false)
   &ScaleColor(
    Scale Alpha=&Lerp_Float(A=1, B=0, Alpha=$Particles.NormalizedAge)
    Scale RGB=&MakeVectorFromLinearColorRGB(Color=&ColorFromCurve(ColorCurve=DataInterface() {propertyValues {
        AlphaCurve {keys []}
        BlueCurve {keys [<at 0 0.25 RCIM_Cubic>,<at 0.492 0.5>,<at 1 0.875 RCIM_Cubic>]}
        GreenCurve {keys [<at 0 0.4375 RCIM_Cubic>,<at 0.492 0.625>,<at 1 0.9062 RCIM_Cubic>]}
        RedCurve {keys [<at 0 1 RCIM_Cubic>,<at 0.492 1>,<at 1 1 RCIM_Cubic>]}
       }}))
    ScaleA=true
    ScaleRGB=true
   )
   &ScaleColor(
    Scale RGB=&VectorFromFloat(Value=&FloatFromCurve(FloatCurve=DataInterface() {propertyValues {Curve {keys [
          <at 0 100 RCIM_Cubic>
          {arriveTangent=-31.4177, interpMode=RCIM_Cubic, leaveTangent=-31.4183, tangentMode=RCTM_User, time=0.25, value=10}
          <at 1 1 RCIM_Cubic>
         ]}}}))
    ScaleA=true
    ScaleRGB=true
   ) {module=ScaleColor001}
   &ScaleSpriteSizeBySpeed(Max Scale Factor=<1 5>)
   &ScaleSpriteSize(Scale Factor=&Vector2DFromFloat(Value=&FloatFromCurve(FloatCurve=DataInterface() {propertyValues {Curve {keys [
          {arriveTangent=42.1764, interpMode=RCIM_Cubic, leaveTangent=42.1763, tangentMode=RCTM_User, time=0, value=0}
          <at 0.03 1 RCIM_Cubic>
          <at 1 0>
         ]}}})))
  ]
  rendererClasses [&NiagaraSpriteRendererProperties]
  renderers [{rendererClass=&NiagaraSpriteRendererProperties}]
 }
 Rim_Glow {
  emitterUpdateScript [
   &EmitterState(Life Cycle Mode=::NewEnumerator1, Loop Behavior=::NewEnumerator1, Loop Delay=0.2, Loop Duration Mode=::NewEnumerator1, UseLoopDelay=true)
   &SpawnBurst_Instantaneous(Spawn Count=4, Use Spawn Probability=false)
  ]
  particleSpawnScript [
   &FN_SetSpawnGroup_v2(Number of Spawn Groups=4)
   &InitializeParticle(
    Color=<rgba 20 6.5 2 0.5>
    Lifetime=5
    Module.Write Color=true
    Module.Write Lifetime=false
    Module.Write Mass=true
    Module.Write Material Random=true
    Module.Write Position=false
    Module.Write Scale=false
    Module.Write SpriteRotation=false
    Module.Write SpriteSize=true
    Position=&SimulationPosition()
    Sprite Size=<5 5>
   )
  ]
  particleUpdateScript [
   &ParticleState(Kill Particles When Lifetime Has Elapsed=false, Let Infinitely Lived Particles Die When Emitter Deactivates=true)
   &SkeletalMeshLocation(
    Apply Sampled Orientation to Mesh Particles=true
    Bone / Socket Index=$Particles.SpawnGroup
    Inherit Velocity=false
    Interpolate Mesh Sampling=false
    MaskSpawnGroup=false
    Mesh Sampling Type=::NewEnumerator5
    Sampled Position Offset=<1 0 0>
    Sampled Position Offset Coordinate Space:ENiagaraMeshTransforms=::NewEnumerator3
    Skeletal Mesh=$User.Skeletal Mesh
    Socket Sampling Mode=::NewEnumerator5
   )
   &FN_Light_v3(
    Color=<rgba 1 0.325 0.1 1>
    Create Color Attribute=true
    Create Exponent Attribute=true
    Create Radius Attribute=true
    Enable Particle Light Enable=true
    Exponent=2
    Light Enabled=true
    Radius=10
    Scale Color=40
   )
  ]
  rendererClasses [&NiagaraSpriteRendererProperties, &NiagaraLightRendererProperties]
  renderers [
   {rendererClass=&NiagaraSpriteRendererProperties}
   {rendererClass=&NiagaraLightRendererProperties, rendererIndex=1}
  ]
 }
 Smoke {
  emitterUpdateScript [
   &EmitterState(Delay First Loop Only=true, Life Cycle Mode=::NewEnumerator1, Loop Delay=0.2, Loop Duration=0.1, UseLoopDelay=true)
   &SpawnRate(SpawnRate=80, Use Spawn Probability=false)
  ]
  particleSpawnScript [
   &FN_SetSpawnGroup_v2(Number of Spawn Groups=4)
   &InitializeParticle(
    Color=<rgba 1 1 1 0.0625>
    Lifetime=&UniformRangedFloat(Minimum=0.45, Maximum=0.65)
    Module.Write Color=true
    Module.Write Lifetime=true
    Module.Write Mass=true
    Module.Write Material Random=true
    Module.Write Position=false
    Module.Write Scale=false
    Module.Write SpriteRotation=true
    Module.Write SpriteSize=true
    Position=&SimulationPosition()
    Sprite Rotation=&NormalizedAngleToDegrees(NormalizedAngle=&UniformRangedFloat(Minimum=0, Maximum=1))
    Sprite Size=&Vector2DFromFloat(Value=&Lerp_Float(
      A=&UniformRangedFloat(Minimum=13, Maximum=25)
      B=&UniformRangedFloat(Minimum=25, Maximum=50)
      Alpha=$User.ForwardDot
     ))
   )
   &DynamicMaterialParameters(
    Index 0 Param 2=1
    Index 0 Param 4=1
    Module.Param0WriteEnabled=false
    Module.Param1WriteEnabled=true
    Module.Param2WriteEnabled=false
    Module.Param3WriteEnabled=true
   )
  ]
  particleUpdateScript [
   &ParticleState()
   &SkeletalMeshLocation(
    Bone / Socket Index=$Particles.SpawnGroup
    Inherit Velocity=true
    MaskSpawnGroup=false
    Mesh Sampling Type=::NewEnumerator5
    Sampled Position Offset Coordinate Space:ENiagaraMeshTransforms=::NewEnumerator3
    Skeletal Mesh=$User.Skeletal Mesh
    Socket Sampling Mode=::NewEnumerator5
    VelocitySpeedLimit=false
   )
   &AddVelocity(Velocity=&RotateVectorByQuaternion(
     Vector To Rotate=&Multiply_Vector(
      A=/BRCosmetics/Athena/Effects/Particles/Events/Papaya/FluorescentLights/FN_SelectAnArrayEntry_Vector(Index:NiagaraInt32=$Particles.SpawnGroup, Vector Array:NiagaraDataInterfaceArrayFloat3=DataInterface() {propertyValues {FloatData []}})
      B=&VectorFromFloat(Value=&self:Glider_Dot_Adjustment(
        Forward Dot:NiagaraFloat=$User.ForwardDot
        Max Value:NiagaraFloat=5
        Min Value:NiagaraFloat=2
        Right Dot:NiagaraFloat=$User.RightDot
        Spawn Group:NiagaraInt32=$Particles.SpawnGroup
       ))
     )
     Quaternion=$Output.SkeletalMeshLocation.SampledRotation
    ))
   &AccelerationForce(Acceleration=<0 0 100>)
   &Drag(Use Linear Drag=true, Use Rotational Drag=false)
   &SolveForcesAndVelocity(Clamp Velocity=false, Limit Acceleration=false)
   &DynamicMaterialParameters(
    Index 0 Param 1=&Lerp_Float(A=1, B=-1, Alpha=$Particles.NormalizedAge)
    Index 0 Param 2=1
    Index 0 Param 4=1
    Module.Param0WriteEnabled=true
    Module.Param1WriteEnabled=false
    Module.Param2WriteEnabled=false
    Module.Param3WriteEnabled=false
   ) {module=DynamicMaterialParameters001}
   /SetVariables(Particles.SubImageIndex=&FloatFromCurve(FloatCurve=DataInterface() {propertyValues {Curve {keys [{arriveTangent=105.1181, interpMode=RCIM_Cubic, leaveTangent=105.1181, tangentMode=RCTM_User, time=0, value=0}, <at 1 63 RCIM_Cubic>]}}})) {isSetParameters=true, module=SetVariables_#1}
   &ScaleColor(
    Scale Alpha=&FloatFromCurve(FloatCurve=DataInterface() {propertyValues {Curve {keys [<at 0 0 RCIM_Cubic>,<at 0.1 1 RCIM_Cubic>,<at 1 0 RCIM_Cubic>]}}})
    Scale RGB=&VectorFromFloat(Value=&FloatFromCurve(FloatCurve=DataInterface() {propertyValues {Curve {keys [<at 0 5 RCIM_Cubic>,<at 0.5 1 RCIM_Cubic>]}}}))
    ScaleA=true
    ScaleRGB=true
   )
   &ScaleColor(
    Scale RGB=&MakeVectorFromLinearColorRGB(Color=&ColorFromCurve(ColorCurve=DataInterface() {propertyValues {
        AlphaCurve {keys []}
        BlueCurve {keys [<at 0 0.25 RCIM_Cubic>,<at 0.6182 0.125 RCIM_Cubic>]}
        GreenCurve {keys [<at 0 0.4375 RCIM_Cubic>,<at 0.6182 0.125 RCIM_Cubic>]}
        RedCurve {keys [<at 0 1 RCIM_Cubic>,<at 0.6182 0.125 RCIM_Cubic>]}
       }}))
    ScaleA=true
    ScaleRGB=true
   ) {module=ScaleColor001}
   &ScaleSpriteSize(Scale Factor=&Vector2DFromFloat(Value=&FloatFromCurve(FloatCurve=DataInterface() {propertyValues {Curve {keys [{arriveTangent=6.2193, interpMode=RCIM_Cubic, leaveTangent=6.2193, tangentMode=RCTM_User, time=0, value=0.2}, <at 1 2 RCIM_Cubic>]}}})))
  ]
  rendererClasses [&NiagaraSpriteRendererProperties]
  renderers [{rendererClass=&NiagaraSpriteRendererProperties}]
 }
 SmokeBurst001 {
  emitterUpdateScript [
   &EmitterState(Life Cycle Mode=::NewEnumerator1, Loop Behavior=::NewEnumerator1, Loop Delay=0.2, Loop Duration=0.1, UseLoopDelay=false)
   &SpawnBurst_Instantaneous(Spawn Count=&UniformRangedInt(Minimum=20, Maximum=40), Use Spawn Probability=false)
  ]
  particleSpawnScript [
   &FN_SetSpawnGroup_v2(Number of Spawn Groups=4)
   &InitializeParticle(
    Color=<rgba 1 1 1 0.0625>
    Lifetime=&UniformRangedFloat(Minimum=0.45, Maximum=0.65)
    Module.Write Color=true
    Module.Write Lifetime=true
    Module.Write Mass=false
    Module.Write Material Random=true
    Module.Write Position=false
    Module.Write Scale=false
    Module.Write SpriteRotation=true
    Module.Write SpriteSize=true
    Position=&SimulationPosition()
    Sprite Rotation=&NormalizedAngleToDegrees(NormalizedAngle=&UniformRangedFloat(Minimum=0, Maximum=1))
    Sprite Size=&Vector2DFromFloat(Value=&Lerp_Float(
      A=&UniformRangedFloat(Minimum=13, Maximum=25)
      B=&UniformRangedFloat(Minimum=25, Maximum=50)
      Alpha=$User.ForwardDot
     ))
   )
   &DynamicMaterialParameters(
    Index 0 Param 2=1
    Index 0 Param 4=1
    Module.Param0WriteEnabled=false
    Module.Param1WriteEnabled=true
    Module.Param2WriteEnabled=false
    Module.Param3WriteEnabled=true
   )
   &SkeletalMeshLocation(
    Bone / Socket Index=$Particles.SpawnGroup
    Inherit Velocity=false
    Interpolate Mesh Sampling=false
    MaskSpawnGroup=false
    Mesh Sampling Type=::NewEnumerator5
    Skeletal Mesh=$User.Skeletal Mesh
    Socket Sampling Mode=::NewEnumerator5
   )
   &AddVelocity(
    Coordinate Space=::Simulation
    Velocity=&RotateVectorByQuaternion(
     Vector To Rotate=&UniformRangedVector(Minimum=<250 -25 -25>, Maximum=<450 25 25>)
     Quaternion=$Output.SkeletalMeshLocation.SampledRotation
    )
   )
  ]
  particleUpdateScript [
   &ParticleState()
   &AccelerationForce(Acceleration=<0 0 100>)
   &Drag(Drag=2, Use Linear Drag=true, Use Rotational Drag=false)
   &SolveForcesAndVelocity(Clamp Velocity=false, Limit Acceleration=false)
   &DynamicMaterialParameters(
    Index 0 Param 1=&Lerp_Float(A=1, B=-1, Alpha=$Particles.NormalizedAge)
    Index 0 Param 2=1
    Index 0 Param 4=1
    Module.Param0WriteEnabled=true
    Module.Param1WriteEnabled=false
    Module.Param2WriteEnabled=false
    Module.Param3WriteEnabled=false
   ) {module=DynamicMaterialParameters001}
   /SetVariables(Particles.SubImageIndex=&FloatFromCurve(FloatCurve=DataInterface() {propertyValues {Curve {keys [{arriveTangent=105.1181, interpMode=RCIM_Cubic, leaveTangent=105.1181, tangentMode=RCTM_User, time=0, value=0}, <at 1 63 RCIM_Cubic>]}}})) {isSetParameters=true, module=SetVariables_#1}
   &ScaleColor(
    Scale Alpha=&FloatFromCurve(FloatCurve=DataInterface() {propertyValues {Curve {keys [<at 0 0 RCIM_Cubic>,<at 0.1 1 RCIM_Cubic>,<at 1 0 RCIM_Cubic>]}}})
    Scale RGB=&VectorFromFloat(Value=&FloatFromCurve(FloatCurve=DataInterface() {propertyValues {Curve {keys [<at 0 5 RCIM_Cubic>,<at 0.5 1 RCIM_Cubic>]}}}))
    ScaleA=true
    ScaleRGB=true
   )
   &ScaleColor(
    Scale RGB=&MakeVectorFromLinearColorRGB(Color=&ColorFromCurve(ColorCurve=DataInterface() {propertyValues {
        AlphaCurve {keys []}
        BlueCurve {keys [<at 0 0.25 RCIM_Cubic>,<at 0.6182 0.125 RCIM_Cubic>]}
        GreenCurve {keys [<at 0 0.4375 RCIM_Cubic>,<at 0.6182 0.125 RCIM_Cubic>]}
        RedCurve {keys [<at 0 1 RCIM_Cubic>,<at 0.6182 0.125 RCIM_Cubic>]}
       }}))
    ScaleA=true
    ScaleRGB=true
   ) {module=ScaleColor001}
   &ScaleSpriteSize(Scale Factor=&Vector2DFromFloat(Value=&FloatFromCurve(FloatCurve=DataInterface() {propertyValues {Curve {keys [{arriveTangent=6.2193, interpMode=RCIM_Cubic, leaveTangent=6.2193, tangentMode=RCTM_User, time=0, value=0.2}, <at 1 2 RCIM_Cubic>]}}})))
  ]
  rendererClasses [&NiagaraSpriteRendererProperties]
  renderers [{rendererClass=&NiagaraSpriteRendererProperties}]
 }
 Spark_Start001 {
  simTarget=GPUComputeSim
  emitterUpdateScript [
   &EmitterState(Life Cycle Mode=::NewEnumerator1, Loop Behavior=::NewEnumerator1, Loop Duration=0.1, UseLoopDelay=false)
   &SpawnBurst_Instantaneous(Spawn Count=&UniformRangedInt(Minimum=40, Maximum=80), Use Spawn Probability=false)
  ]
  particleSpawnScript [
   &FN_SetSpawnGroup_v2(Number of Spawn Groups=4)
   &InitializeParticle(
    Color=<rgba 10 5.25 2.5 1>
    Lifetime=&UniformRangedFloat(Minimum=0.12, Maximum=0.4)
    Module.Write Color=true
    Module.Write Lifetime=true
    Module.Write Mass=true
    Module.Write Material Random=true
    Module.Write Position=false
    Module.Write Scale=false
    Module.Write SpriteRotation=false
    Module.Write SpriteSize=true
    Position=&SimulationPosition()
    Sprite Size=&UniformRangedVector2D(Minimum=<0.7 0.5>, Maximum=<1.2 1.5>)
   )
   &SkeletalMeshLocation(
    Bone / Socket Index=$Particles.SpawnGroup
    Inherit Velocity=false
    Interpolate Mesh Sampling=false
    MaskSpawnGroup=false
    Mesh Sampling Type=::NewEnumerator5
    Sampled Position Offset Coordinate Space:ENiagaraMeshTransforms=::NewEnumerator3
    Skeletal Mesh=$User.Skeletal Mesh
    Socket Sampling Mode=::NewEnumerator5
   )
   &SphereLocation(
    Hemisphere X=true
    MaskSpawnGroup=false
    Module.Override Randomness=false
    Sphere Radius=&UniformRangedFloat(Minimum=5, Maximum=15)
    Surface Only=false
   )
   &AddVelocity(Velocity=&RotateVectorByQuaternion(
     Vector To Rotate=&Multiply_Vector(
      A=$Output.SphereLocation.SphereNormal
      B=&VectorFromFloat(Value=&UniformRangedFloat(Minimum=25, Maximum=55))
     )
     Quaternion=$Output.SkeletalMeshLocation.SampledRotation
    )) {module=AddVelocity001}
   &AddVelocity(Velocity=&UniformRangedVector(Minimum=<-100 -100 -150>, Maximum=<100 100 -80>))
  ]
  particleUpdateScript [
   &ParticleState()
   &AccelerationForce(Acceleration=<0 0 -1200>)
   &SampleVectorField(
    Sampled Vector Scale=0.4
    VectorField=DataInterface() {propertyValues {Field=&Noisy_Turbulence_103, bTileX=true, bTileY=true, bTileZ=true}}
   )
   &ApplyVectorField(Vector Field Intensity as Force=0, Vector Field Intensity as Velocity=250, Vector Field Velocity Tightness=0.15)
   &SolveForcesAndVelocity(Clamp Velocity=false, Limit Acceleration=false)
   &ScaleColor(
    Scale Alpha=&FloatFromCurve(FloatCurve=DataInterface() {propertyValues {Curve {keys [<at 0 1>,<at 1 0>]}}})
    Scale RGB=&MakeVectorFromLinearColorRGB(Color=&ColorFromCurve(ColorCurve=DataInterface() {propertyValues {
        AlphaCurve {keys []}
        BlueCurve {keys [<at 0 0.25 RCIM_Cubic>,<at 0.5 0.5>,<at 1 0.875 RCIM_Cubic>]}
        GreenCurve {keys [<at 0 0.525 RCIM_Cubic>,<at 0.5 0.6833>,<at 1 0.9208 RCIM_Cubic>]}
        RedCurve {keys [<at 0 1 RCIM_Cubic>,<at 0.5 1>,<at 1 1 RCIM_Cubic>]}
       }}))
    ScaleA=true
    ScaleRGB=true
   )
   &ScaleColor(
    Scale RGB=&VectorFromFloat(Value=&FloatFromCurve(FloatCurve=DataInterface() {propertyValues {Curve {keys [
          <at 0 100 RCIM_Cubic>
          {arriveTangent=-36.3579, interpMode=RCIM_Cubic, leaveTangent=-36.3595, tangentMode=RCTM_User, time=0.25, value=10}
          <at 1 1 RCIM_Cubic>
         ]}}}))
    ScaleA=true
    ScaleRGB=true
   ) {module=ScaleColor001}
   &ScaleSpriteSizeBySpeed(Max Scale Factor=<1 8>, Min Scale Factor=<1 0.5>)
   &ScaleSpriteSize(Scale Factor=&Vector2DFromFloat(Value=&FloatFromCurve(FloatCurve=DataInterface() {propertyValues {Curve {keys [
          {arriveTangent=43.4604, interpMode=RCIM_Cubic, leaveTangent=43.4604, tangentMode=RCTM_User, time=0.0001, value=0.0578}
          {arriveTangent=-0.0578, interpMode=RCIM_Cubic, leaveTangent=-0.0578, time=0.03, value=1}
          {arriveTangent=-3.2182, interpMode=RCIM_Cubic, leaveTangent=-3.2182, tangentMode=RCTM_User, time=1, value=0}
         ]}}})))
  ]
  rendererClasses [&NiagaraSpriteRendererProperties]
  renderers [{rendererClass=&NiagaraSpriteRendererProperties}]
 }
}
summary {
 systemName=NS_Glider_HighTower_Tomato_Player
 emitters [
  {emitterName=Flame_Glow, rendererClasses [&NiagaraSpriteRendererProperties]}
  {emitterName=Smoke, rendererClasses [&NiagaraSpriteRendererProperties]}
  {emitterName=Random_Sparks, simTarget=GPUComputeSim, rendererClasses [&NiagaraSpriteRendererProperties]}
  {emitterName=Rim_Glow, rendererClasses [&NiagaraSpriteRendererProperties, &NiagaraLightRendererProperties]}
  {
   emitterName=Animorphic_-_HotCoreStart_-_FlameOuterStart
   rendererClasses [&NiagaraSpriteRendererProperties, &NiagaraMeshRendererProperties]
  }
  {emitterName=Spark_Start001, simTarget=GPUComputeSim, rendererClasses [&NiagaraSpriteRendererProperties]}
  {emitterName=SmokeBurst001, rendererClasses [&NiagaraSpriteRendererProperties]}
  {emitterName=Comet_Flame_Meshes, rendererClasses [&NiagaraMeshRendererProperties]}
 ]
 userVariables [
  User.ForwardDot {type=NiagaraFloat, defaultValue=0, description=""}
  User.RightDot {type=NiagaraFloat, defaultValue=0, description=""}
  User.Skeletal Mesh {
   type=NiagaraDataInterfaceSkeletalMesh
   defaultValue {
    struct=NiagaraExt_VariableValue_DataInterface
    value {dataInterface=&self:NiagaraDataInterfaceSkeletalMesh_0}
   }
  }
 ]
}
